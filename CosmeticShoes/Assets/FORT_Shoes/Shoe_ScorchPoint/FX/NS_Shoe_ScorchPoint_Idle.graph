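# Idle Niagara effect for the ScorchPoint cosmetic shoes, dressing both feet in flames, smoke, embers and glow.
# Sprite embers and flames, mesh flames and two glow lights attach to skeletal-mesh bones or sockets, scaling per foot by User.ShoeScale_L/R.
# System timelines drive intensity, HideFootFX fades elements out, and isOffPawnPreview switches most to looping preview curves.
# Embers stop spawning while falling; the texture-masked GPU_Flames emitter is disabled.
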NiagaraSystem /CosmeticShoes/Assets/FORT_Shoes/Shoe_ScorchPoint/FX/NS_Shoe_ScorchPoint_Idle
dependencies {
 usedDataInterfaces [
  &NiagaraDataInterfaceCurve
  &NiagaraDataInterfaceColorCurve
  &NiagaraDataInterfaceSpriteRendererInfo
  /Script/Niagara.NiagaraDataInterfaceTexture
  &NiagaraDataInterfaceArrayInt32
  &NiagaraDataInterfaceArrayFloat3
  &NiagaraDataInterfaceArrayFloat
  &NiagaraDataInterfaceSkeletalMesh
 ]
 usedDynamicInputs [
  &Boolean_AndOperation
  &VectorLength
  &Multiply_Vector
 ]
 usedModules [
  &self:SystemSpawnScript.SystemScriptSource.SystemScriptGraph.NiagaraNodeAssignment_0.SetVariables_#1
  &SystemState
  &self:SystemSpawnScript.SystemScriptSource.SystemScriptGraph.NiagaraNodeAssignment_1.SetVariables_#2
  &self:SystemSpawnScript.SystemScriptSource.SystemScriptGraph.NiagaraNodeAssignment_2.SetVariables_#3
  &Timeline
  &self:SystemSpawnScript.SystemScriptSource.SystemScriptGraph.NiagaraNodeAssignment_4.SetVariables_#4
 ]
}
emitters {
 Embers {
  emitterUpdateScript [
   &EmitterState()
   &SpawnRate(
    Spawn Probability=0.6
    SpawnRate=&MakeCustomFloatFromBool(
     Bool=$User.isOffPawnPreview
     True Float=$System.FELoop
     False Float=&MakeCustomFloatFromBool(
      Bool=$System.isFalling
      True Float=0
      False Float=&Multiply_Float(A=&Lerp_Float(A=12, B=30, Alpha=$System.Timeline001.Time), B=$System.Timeline.Time)
     )
    )
    Use Spawn Probability=true
   )
  ]
  particleSpawnScript [
   &PartitionParticles(Number Of Partitions=20, Partition Mode:ENiagara_PartitionMode_v2=::NewEnumerator5)
   &InitializeParticle(
    Color=<rgba 2 0.5898 0 1>
    Lifetime Max=0.75
    Lifetime Min=0.35
    Lifetime Mode=::NewEnumerator1
    Sprite Size Mode=::NewEnumerator3
    Uniform Sprite Size=&Multiply_Float(
     A=&RandomRangeFloat(Minimum=5, Maximum=9)
     B=&MakeCustomFloatFromBool(Bool=$User.isOffPawnPreview, True Float=0.5, False Float=1)
    )
    UsePositionOffset=false
    Write Material Random=true
   )
   &SkeletalMeshLocation(
    Bone / Socket Index=$Particles.PartitionParticles.Partition
    MaskSpawnGroup=false
    Mesh Sampling Type=::NewEnumerator5
    Skeletal Mesh=$System.Skeletal Mesh
    Socket Sampling Mode=::NewEnumerator5
   )
   &AddVelocity(Velocity=&RandomRangeVector(Minimum=<0 0 35>, Maximum=<0 0 60>))
   &AddVelocity(
    Module.Use Falloff Distance=false
    Override Seed=false
    Velocity Mode=::NewEnumerator1
    Velocity Speed=&RandomRangeFloat(Minimum=10, Maximum=20)
   ) {module=AddVelocity001}
  ]
  particleUpdateScript [
   &ParticleState()
   &ScaleSpriteSize(Uniform Curve Sprite Scale=DataInterface() {propertyValues {Curve {keys [{tangentMode=RCTM_User, time=0, value=0}, <at 0.2 1>, <at 0.7 1>, {tangentMode=RCTM_User, time=1, value=0.25}]}}})
   &ScaleColor(
    Linear Color Curve=DataInterface() {propertyValues {
      AlphaCurve {keys [<at 0 1>,<at 0.9796 0>]}
      BlueCurve {keys [<at 0 0.9356>,<at 0.5239 0.0215>,<at 0.7978 0>]}
      GreenCurve {keys [<at 0 1.2022>,<at 0.5239 0.0062>,<at 0.7978 0.0018>]}
      RedCurve {keys [<at 0 2>,<at 0.5239 0.1156>,<at 0.7978 0.0698>]}
     }}
    Scale Mode=::NewEnumerator2
   )
   &ScaleColor(
    Scale Alpha=&MakeCustomFloatFromBool(Bool=$User.HideFootFX, True Float=0, False Float=1)
    ScaleA=true
    ScaleRGB=false
   ) {module=ScaleColor001}
   &Drag(Drag=&RandomRangeFloat(Minimum=3, Maximum=6), Use Linear Drag=true, Use Rotational Drag=true)
   &CurlNoiseForce(
    Bias Noise=false
    Noise Frequency=10
    Noise Quality / Cost=::NewEnumerator0
    Noise Strength=&RandomRangeFloat(Minimum=10, Maximum=20)
    Pan Noise=false
   )
   &SolveForcesAndVelocity(Clamp Velocity=false, Limit Acceleration=false)
   &DynamicMaterialParameters(
    Index 0 Param 1=&FloatFromCurve(FloatCurve=DataInterface() {propertyValues {Curve {keys [
         {interpMode=RCIM_Cubic, leaveTangent=2.7576, tangentMode=RCTM_Break, time=0.2, value=0}
         {arriveTangent=2.8876, tangentMode=RCTM_User, time=0.7, value=1}
        ]}}})
    Index 0 Param 2=&RandomRangeFloat(Minimum=0, Maximum=1)
    Index 0 Param 3=&RandomRangeFloat(Minimum=0, Maximum=1)
    Module.Index0Float0Write=true
    Module.Index0Float1Write=true
    Module.Index0Float2Write=true
    Module.Index0Float3Write=true
   )
   &ScaleSpriteSize(
    Scale Sprite Size Mode=::NewEnumerator0
    Uniform Scale Factor=&MakeCustomFloatFromBool(
     Bool=&SetBoolByIntComparison(A=$Particles.PartitionParticles.Partition, B=10, Comparison Type=::NewEnumerator6)
     True Float=&MakeFloatFromVector(Vector=$User.ShoeScale_R, Channel=::NewEnumerator0)
     False Float=&MakeFloatFromVector(Vector=$User.ShoeScale_L, Channel=::NewEnumerator0)
    )
   ) {module=ScaleSpriteSize001}
  ]
  rendererClasses [&NiagaraSpriteRendererProperties]
  renderers [{rendererClass=&NiagaraSpriteRendererProperties}]
 }
 GPU_Flames {
  bEnabled=false
  simTarget=GPUComputeSim
  emitterUpdateScript [
   &EmitterState()
   &SpawnRate(
    SpawnRate=&MakeCustomFloatFromBool(
     Bool=$User.isOffPawnPreview
     True Float=&FloatFromCurve(
      FloatCurve=DataInterface() {propertyValues {Curve {keys [
          {interpMode=RCIM_Cubic, leaveTangent=3.1595, tangentMode=RCTM_Break, time=0, value=0}
          {arriveTangent=0.2069, interpMode=RCIM_Cubic, leaveTangent=0.2069, tangentMode=RCTM_User, time=0.2, value=1}
          {interpMode=RCIM_Cubic, tangentMode=RCTM_Break, time=0.5, value=1}
          {arriveTangent=-3.2253, tangentMode=RCTM_User, time=0.8, value=0}
         ]}}}
      CurveIndex=$System.NormalizedLoopAge
      Scale Curve=30
     )
     False Float=&Multiply_Float(A=70, B=$System.Timeline.Time)
    )
    Use Spawn Probability=false
   )
  ]
  particleSpawnScript [&InitializeParticle(
    Color=<rgba 1 0.4239 0.3889 1>
    Lifetime Min=1.25
    Lifetime Mode=::NewEnumerator1
    Sprite Size Max=<10 12>
    Sprite Size Mode=::NewEnumerator2
    UsePositionOffset=false
    Write Material Random=true
   )]
  particleUpdateScript [
   &ParticleState()
   &ScaleSpriteSize(Uniform Curve Sprite Scale=DataInterface() {propertyValues {Curve {keys [
        {tangentMode=RCTM_User, time=0, value=0.3118}
        {interpMode=RCIM_Cubic, leaveTangent=1.2886, tangentMode=RCTM_Break, time=0.1504, value=0.8892}
        {tangentMode=RCTM_User, time=0.9986, value=1.3647}
       ]}}})
   &SkeletalMeshLocation(
    MaskSpawnGroup=false
    Mesh Sampling Type=::NewEnumerator7
    OffsetAlongNormal=false
    OverrideRandomSeed=false
    Skeletal Mesh=$System.Skeletal Mesh
   )
   &SpriteFacingAndAlignment(Sprite Alignment=<0 0 1>, Sprite Facing=$Particles.SkeletalMeshLocation.SampledNormal, WriteAlignment=true, WriteFacing=false)
   &SubUVAnimation(
    Play Rate=0.7
    Random Start Frame=true
    "Sprite Renderer "=DataInterface() {propertyValues {SpriteRenderer=&self:GPU_Flames_0.NiagaraSpriteRendererProperties_0}}
    SubUV Animation Mode:ENiagara_SubUVLookupModeV2=::NewEnumerator3
    UseEndFrame=false
    UsePlayRateScale=false
    UseStartFrame=false
   ) {enabled=false}
   /Niagara/Modules/Update/Textures/SampleTexture(
    Apply Sampled RGBA to Particle Color=false
    Texture=DataInterface() {propertyValues {
      Texture=&/T_Shoe_ScorchPoint_VFX_Mask
      TextureUserParameter {parameter {name=None, typeDefHandle {registeredTypeIndex=2}, varData []}}
     }}
    UV:Vector2f=$Particles.SkeletalMeshLocation.SampledUV
   )
   &KillParticles(Kill Particles=&SetBoolByFloatComparison(A=$Particles.SampleTexture.SampledB, B=0.01, Comparison Type=::NewEnumerator6))
   &CameraOffset(Camera Offset Amount=2)
   &DynamicMaterialParameters(
    Index 0 Param 1=&UniformAOrBFloat(A=1, B=2, Evaluation Type=::NewEnumerator0, Recalculate Random Each Loop=true, Override Seed=false)
    Index 0 Param 3=8
    Index 0 Param 4=&FloatFromCurve(
     FloatCurve=DataInterface() {propertyValues {Curve {keys [
         {interpMode=RCIM_Cubic, leaveTangent=0.0464, tangentMode=RCTM_Break, time=0, value=0}
         {arriveTangent=2.3089, tangentMode=RCTM_User, time=1, value=1}
        ]}}}
     Scale Curve=64
    )
    Index 1 Param 1=&RandomRangeFloat(Minimum=0, Maximum=1)
    Index 1 Param 2=&RandomRangeFloat(Minimum=0, Maximum=1)
    Index 1 Param 3=&FloatFromCurve(
     FloatCurve=DataInterface() {propertyValues {Curve {keys [{interpMode=RCIM_Cubic, leaveTangent=-2.2926, tangentMode=RCTM_Break, time=0, value=1}, <at 1 0>]}}}
     Scale Curve=5
    )
    Module.Index0Float0Write=true
    Module.Index0Float1Write=true
    Module.Index0Float2Write=true
    Module.Index0Float3Write=true
    Module.Index1Float0Write=true
    Module.Index1Float1Write=true
    Module.Index1Float2Write=true
    Module.Index1Float3Write=true
    Write Parameter Index 1=::NewEnumerator1
   )
   &ScaleColor(Scale Alpha=$System.Timeline.Time, ScaleA=true, ScaleRGB=true)
   /SetVariables(
    Particles.GlowColor=&ColorFromCurve(ColorCurve=DataInterface() {propertyValues {
       AlphaCurve {keys [<at 0.048 0>,<at 0.3162 0.4 RCIM_Cubic>,<at 0.9481 0 RCIM_Cubic>]}
       BlueCurve {keys [<at 0 0.123 RCIM_Cubic>,<at 1 0.0436 RCIM_Cubic>]}
       GreenCurve {keys [<at 0 0.1453 RCIM_Cubic>,<at 1 0.0451 RCIM_Cubic>]}
       RedCurve {keys [<at 0 1 RCIM_Cubic>,<at 1 0.4844 RCIM_Cubic>]}
      }})
    Particles.GlowDynPar:Vector4f=<8 0 0 0>
    Particles.GlowSize=<4 8>
   ) {isSetParameters=true, module=SetVariables_#5}
  ]
  rendererClasses [&NiagaraSpriteRendererProperties]
  renderers [
   {rendererClass=&NiagaraSpriteRendererProperties}
   {rendererClass=&NiagaraSpriteRendererProperties, rendererIndex=1}
  ]
 }
 GPU_Flames002 {
  emitterUpdateScript [
   &EmitterState()
   &SpawnRate(
    SpawnRate=&MakeCustomFloatFromBool(
     Bool=$User.isOffPawnPreview
     True Float=$System.FELoop
     False Float=&Multiply_Float(A=15, B=$System.Timeline.Time)
    )
    Use Spawn Probability=false
   )
  ]
  particleSpawnScript [
   &PartitionParticles(Number Of Partitions=20, Partition Mode:ENiagara_PartitionMode_v2=::NewEnumerator5)
   &InitializeParticle(
    Color=<rgba 1 0.4239 0.3889 1>
    Lifetime Min=1.25
    Lifetime Mode=::NewEnumerator1
    Sprite Size Max=<10 12>
    Sprite Size Mode=::NewEnumerator2
    UsePositionOffset=false
    Write Material Random=true
   )
  ]
  particleUpdateScript [
   &ParticleState()
   &ScaleSpriteSize(Uniform Curve Sprite Scale=DataInterface() {propertyValues {Curve {keys [
        {tangentMode=RCTM_User, time=0, value=0.3118}
        {interpMode=RCIM_Cubic, leaveTangent=1.2886, tangentMode=RCTM_Break, time=0.1504, value=0.8892}
        {tangentMode=RCTM_User, time=0.9986, value=1.3647}
       ]}}})
   &SkeletalMeshLocation(
    Bone / Socket Index=$Particles.PartitionParticles.Partition
    MaskSpawnGroup=false
    Mesh Sampling Type=::NewEnumerator5
    Skeletal Mesh=$System.Skeletal Mesh
    Socket Sampling Mode=::NewEnumerator5
   )
   &SpriteFacingAndAlignment(Sprite Alignment=<0 0 1>, Sprite Facing=$Particles.SkeletalMeshLocation.SampledNormal, WriteAlignment=true, WriteFacing=false)
   &SubUVAnimation(
    Play Rate=0.7
    Random Start Frame=true
    "Sprite Renderer "=DataInterface() {propertyValues {SpriteRenderer=&self:GPU_Flames002_0.NiagaraSpriteRendererProperties_0}}
    SubUV Animation Mode:ENiagara_SubUVLookupModeV2=::NewEnumerator3
    UseEndFrame=false
    UsePlayRateScale=false
    UseStartFrame=false
   ) {enabled=false}
   &SampleTexture(
    Apply Sampled RGBA to Particle Color=false
    Texture=DataInterface() {propertyValues {
      Texture=&T_Shoe_ScorchPoint_VFX_Mask
      TextureUserParameter {parameter {name=None, typeDefHandle {registeredTypeIndex=2}, varData []}}
     }}
    UV:Vector2f=$Particles.SkeletalMeshLocation.SampledUV
   ) {enabled=false}
   &KillParticles(Kill Particles=&SetBoolByFloatComparison(A=$Particles.SampleTexture.SampledB, B=0.01, Comparison Type=::NewEnumerator6)) {enabled=false}
   &CameraOffset(Camera Offset Amount=2)
   &DynamicMaterialParameters(
    Index 0 Param 1=&UniformAOrBFloat(A=1, B=2, Evaluation Type=::NewEnumerator0, Recalculate Random Each Loop=true, Override Seed=false)
    Index 0 Param 3=8
    Index 0 Param 4=&FloatFromCurve(
     FloatCurve=DataInterface() {propertyValues {Curve {keys [
         {interpMode=RCIM_Cubic, leaveTangent=0.0464, tangentMode=RCTM_Break, time=0, value=0}
         {arriveTangent=2.3089, tangentMode=RCTM_User, time=1, value=1}
        ]}}}
     Scale Curve=64
    )
    Index 1 Param 1=&RandomRangeFloat(Minimum=0, Maximum=1)
    Index 1 Param 2=&RandomRangeFloat(Minimum=0, Maximum=1)
    Index 1 Param 3=&FloatFromCurve(
     FloatCurve=DataInterface() {propertyValues {Curve {keys [{interpMode=RCIM_Cubic, leaveTangent=-2.2926, tangentMode=RCTM_Break, time=0, value=1}, <at 1 0>]}}}
     Scale Curve=5
    )
    Module.Index0Float0Write=true
    Module.Index0Float1Write=true
    Module.Index0Float2Write=true
    Module.Index0Float3Write=true
    Module.Index1Float0Write=true
    Module.Index1Float1Write=true
    Module.Index1Float2Write=true
    Module.Index1Float3Write=true
    Write Parameter Index 1=::NewEnumerator1
   )
   &ScaleColor(
    Scale Alpha=&MakeCustomFloatFromBool(
     Bool=$User.isOffPawnPreview
     True Float=1
     False Float=&MakeCustomFloatFromBool(Bool=$User.HideFootFX, True Float=0, False Float=$System.Timeline.Time)
    )
    ScaleA=true
    ScaleRGB=true
   )
   /SetVariables(
    Particles.GlowColor=&ColorFromCurve(
     ColorCurve=DataInterface() {propertyValues {
       AlphaCurve {keys [<at 0.0613 0>,<at 0.3357 0.02>,<at 0.5092 0.4 RCIM_Cubic>,<at 0.8391 0 RCIM_Cubic>]}
       BlueCurve {keys [<at 0 0.123 RCIM_Cubic>,<at 1 0.0436 RCIM_Cubic>]}
       GreenCurve {keys [<at 0 0.1453 RCIM_Cubic>,<at 1 0.0451 RCIM_Cubic>]}
       RedCurve {keys [<at 0 1 RCIM_Cubic>,<at 1 0.4844 RCIM_Cubic>]}
      }}
     CurveIndex=&MakeCustomFloatFromBool(
      Bool=&Boolean_OrOperation(Bool1=$User.HideFootFX, Bool2=$User.isOffPawnPreview)
      True Float=0
      False Float=$Particles.NormalizedAge
     )
    )
    Particles.GlowDynPar:Vector4f=<8 0 0 0>
    Particles.GlowSize=&Multiply_Vector2D_ByFloat(
     Float=&MakeCustomFloatFromBool(
      Bool=&SetBoolByIntComparison(A=$Particles.PartitionParticles.Partition, B=10, Comparison Type=::NewEnumerator6)
      True Float=&MakeFloatFromVector(Vector=$User.ShoeScale_R, Channel=::NewEnumerator0)
      False Float=&MakeFloatFromVector(Vector=$User.ShoeScale_L, Channel=::NewEnumerator0)
     )
     Vector2D=<4 8>
    )
   ) {isSetParameters=true, module=SetVariables_#5}
   &ScaleSpriteSize(
    Scale Sprite Size Mode=::NewEnumerator0
    Uniform Scale Factor=&MakeCustomFloatFromBool(
     Bool=&SetBoolByIntComparison(A=$Particles.PartitionParticles.Partition, B=10, Comparison Type=::NewEnumerator6)
     True Float=&MakeFloatFromVector(Vector=$User.ShoeScale_R, Channel=::NewEnumerator0)
     False Float=&MakeFloatFromVector(Vector=$User.ShoeScale_L, Channel=::NewEnumerator0)
    )
   ) {module=ScaleSpriteSize001}
  ]
  rendererClasses [&NiagaraSpriteRendererProperties]
  renderers [
   {rendererClass=&NiagaraSpriteRendererProperties}
   {rendererClass=&NiagaraSpriteRendererProperties, rendererIndex=1}
  ]
 }
 GlowAndLights {
  emitterUpdateScript [
   &EmitterState(
    Inactive Response=::NewEnumerator1
    Life Cycle Mode=::NewEnumerator1
    Loop Behavior=::NewEnumerator1
    Loop Duration Mode=::NewEnumerator1
    UseLoopDelay=false
   )
   &SpawnBurst_Instantaneous(Spawn Count=2, Use Spawn Probability=false, UseLoopCountLimit=false)
  ]
  particleSpawnScript [&InitializeParticle(
    Color=<rgba 2 0.2453 0.1839 0.1>
    Sprite Size=<25 45>
    Sprite Size Mode=::NewEnumerator4
    UsePositionOffset=false
    Write Material Random=true
   )]
  particleUpdateScript [
   &SolveForcesAndVelocity(Clamp Velocity=false, Limit Acceleration=false)
   &ParticleState(Kill Particles When Lifetime Has Elapsed=false)
   &SkeletalMeshLocation(
    Bone / Socket Index=&SelectIntFromArray(Direct Array Index=&ReturnExecIndex(), Int Selection Array=DataInterface() {propertyValues {IntData [2,3]}})
    Bone Sampling Mode=::NewEnumerator4
    MaskSpawnGroup=false
    Sampled Position Offset=&SelectVectorFromArray(
     Direct Array Index=&ReturnExecIndex()
     Vector Selection Array=DataInterface() {propertyValues {FloatData [<5 0 0>, <-5 0 0>]}}
    )
    Skeletal Mesh=$System.Skeletal Mesh
   )
   &SpriteFacingAndAlignment(
    Sprite Alignment=&RotateVectorByQuaternion(Vector To Rotate=<1 0 0>, Quaternion=$Particles.SkeletalMeshLocation.SampledRotation)
    Sprite Facing=<0 0 1>
    WriteAlignment=true
    WriteFacing=true
   ) {enabled=false}
   &FN_Light_Attributes(
    Color=<rgba 15 3.0977 1.4988 1>
    Create Color Attribute=true
    Create Exponent Attribute=true
    Create Radius Attribute=true
    Exponent=2
    Radius=&Multiply_Float(
     A=30
     B=&MaskFloat(
      Mask Source=1
      Mask Parameter=0
      Mask=&ReturnExecIndex()
      Masked Float=&MakeFloatFromVector(Vector=$User.ShoeScale_R, Channel=::NewEnumerator0)
      Passthrough Float=&MakeFloatFromVector(Vector=$User.ShoeScale_L, Channel=::NewEnumerator0)
     )
    )
    Scale Color=&MakeCustomFloatFromBool(Bool=$User.HideFootFX, True Float=0, False Float=$System.Timeline.Time)
   )
   &DynamicMaterialParameters(
    Index 0 Param 1=4
    Module.Index0Float0Write=true
    Module.Index0Float1Write=true
    Module.Index0Float2Write=true
    Module.Index0Float3Write=true
   )
   &ScaleColor(
    Scale Alpha=&MakeCustomFloatFromBool(
     Bool=$User.isOffPawnPreview
     True Float=&FloatFromCurve(
      FloatCurve=DataInterface() {propertyValues {Curve {keys [
          {interpMode=RCIM_Cubic, leaveTangent=3.1595, tangentMode=RCTM_Break, time=0, value=0}
          {arriveTangent=0.2069, interpMode=RCIM_Cubic, leaveTangent=0.2069, tangentMode=RCTM_User, time=0.2, value=1}
          {interpMode=RCIM_Cubic, tangentMode=RCTM_Break, time=0.5, value=1}
          {arriveTangent=-3.2253, tangentMode=RCTM_User, time=0.8, value=0}
         ]}}}
      CurveIndex=$System.NormalizedLoopAge
     )
     False Float=$System.Timeline.Time
    )
    ScaleA=true
    ScaleRGB=true
   )
   &ScaleColor(
    Scale Alpha=&MakeCustomFloatFromBool(
     Bool=&Boolean_OrOperation(Bool1=$User.HideFootFX, Bool2=$User.isOffPawnPreview)
     True Float=0
     False Float=&MaskFloat(
      Mask Source=1
      Mask Parameter=0
      Mask=&ReturnExecIndex()
      Masked Float=&MakeFloatFromVector(Vector=$User.ShoeScale_R, Channel=::NewEnumerator0)
      Passthrough Float=&MakeFloatFromVector(Vector=$User.ShoeScale_L, Channel=::NewEnumerator0)
     )
    )
    ScaleA=true
    ScaleRGB=true
   ) {module=ScaleColor001}
   &AlignSpriteToMeshOrientation(
    Mesh Orientation Relative Sprite Alignment Vector=<1 0 0>
    Mesh Orientation Relative Sprite Facing Vector:Vector3f=&SelectVectorFromArray(
     Direct Array Index=&ReturnExecIndex()
     Vector Selection Array=DataInterface() {propertyValues {FloatData [<0 1 0>, <0 -1 0>]}}
    )
    SetAlignment=true
    SetFacing=true
   )
   /SetVariables(Particles.LightSpecScale:NiagaraFloat=&FN_FrontEndFloatModifier(Gameplay=1, Front End=0)) {isSetParameters=true}
  ]
  rendererClasses [&NiagaraLightRendererProperties, &NiagaraSpriteRendererProperties]
  renderers [
   {rendererClass=&NiagaraLightRendererProperties, rendererIndex=0}
   {rendererClass=&NiagaraSpriteRendererProperties, rendererIndex=1}
  ]
 }
 SplineFlames {
  emitterUpdateScript [
   &EmitterState()
   &SpawnRate(
    SpawnRate=&MakeCustomFloatFromBool(
     Bool=$User.isOffPawnPreview
     True Float=&FloatFromCurve(
      FloatCurve=DataInterface() {propertyValues {Curve {keys [
          {interpMode=RCIM_Cubic, leaveTangentWeight=5.5233, tangentMode=RCTM_Break, time=0, value=0}
          {arriveTangentWeight=1.702, interpMode=RCIM_Cubic, leaveTangentWeight=3.2337, tangentMode=RCTM_User, time=1, value=0}
          {arriveTangentWeight=0.335, interpMode=RCIM_Cubic, leaveTangentWeight=3.015, tangentMode=RCTM_User, time=2, value=1}
          {arriveTangentWeight=3, interpMode=RCIM_Cubic, leaveTangentWeight=1.6667, tangentMode=RCTM_Break, time=3, value=1}
          {arriveTangent=-1.327, tangentMode=RCTM_User, time=4.5, value=0}
          <at 5 0>
         ]}}}
      CurveIndex=&ModuloFloat(A=&Divide_Float(A=$Engine.Time, B=4), B=6)
      Scale Curve=6
     )
     False Float=6
    )
    Use Spawn Probability=false
   )
  ]
  particleSpawnScript [
   &PartitionParticles(Number Of Partitions=20, Partition Mode:ENiagara_PartitionMode_v2=::NewEnumerator5)
   &InitializeParticle(
    Color=<rgba 1 0.2359 0.4907 1>
    Lifetime Max=2.4
    Lifetime Mode=::NewEnumerator1
    Mesh Scale=<0.5 0.5 0.65>
    Mesh Scale Mode=::NewEnumerator4
    UsePositionOffset=false
    Write Material Random=true
   )
   &InitialMeshOrientation(
    Mesh Orientation Mode=::NewEnumerator2
    Rotation=&SelectVectorFromArray(
     Direct Array Index=$Particles.PartitionParticles.Partition
     Vector Selection Array=DataInterface() {propertyValues {FloatData [<-0.25 0 0>, <0.25 0 0>, <0.25 0 0>, <-0.25 0 0>, <0 0.25 0>, <0 0.25 0>]}}
    )
    Use Rotation Vector=true
   ) {enabled=false}
   /SetVariables(Particles.SmokeSize:Vector3f=<0.5 0.5 0.6>) {isSetParameters=true, module=SetVariables_#6}
  ]
  particleUpdateScript [
   &ParticleState()
   &ScaleMeshSize(Scale Factor=&MakeVector(
     X=1
     Y=1
     Z=&FloatFromCurve(FloatCurve=DataInterface() {propertyValues {Curve {keys [
          <at -1 0>
          {interpMode=RCIM_Cubic, leaveTangent=1.162, tangentMode=RCTM_Break, time=0, value=0}
          {arriveTangent=-0.2693, tangentMode=RCTM_User, time=0.6, value=0.95}
         ]}}})
    ))
   &SkeletalMeshLocation(
    Bone / Socket Index=$Particles.PartitionParticles.Partition
    MaskSpawnGroup=false
    Mesh Sampling Type=::NewEnumerator5
    Sampled Position Offset=<0 0 -2>
    Skeletal Mesh=$System.Skeletal Mesh
    Socket Sampling Mode=::NewEnumerator5
   )
   &DynamicMaterialParameters(
    Index 0 Param 1=1
    Index 0 Param 2=&FloatFromCurve(
     FloatCurve=DataInterface() {propertyValues {Curve {keys [
         {interpMode=RCIM_Cubic, leaveTangent=0.8901, tangentMode=RCTM_Break, time=0, value=0}
         {arriveTangent=0.7138, tangentMode=RCTM_User, time=0.6, value=1}
        ]}}}
     Scale Curve=&Multiply_Float(
      A=&RandomRangeFloat(Minimum=7, Maximum=10)
      B=&MakeCustomFloatFromBool(Bool=$User.isOffPawnPreview, True Float=0.7, False Float=1)
     )
    )
    Index 0 Param 3=&RandomRangeFloat(Minimum=0, Maximum=1)
    Module.Index0Float0Write=true
    Module.Index0Float1Write=true
    Module.Index0Float2Write=true
    Module.Index0Float3Write=true
   )
   &ScaleColor(
    Scale Alpha=&FloatFromCurve(
     FloatCurve=DataInterface() {propertyValues {Curve {keys [
         {arriveTangent=8.0375, interpMode=RCIM_Cubic, leaveTangent=8.0375, tangentMode=RCTM_User, time=0, value=0}
         {interpMode=RCIM_Cubic, leaveTangent=-2.1697, tangentMode=RCTM_Break, time=0.3, value=1}
         <at 1 0>
        ]}}}
     Scale Curve=&MakeCustomFloatFromBool(Bool=$User.isOffPawnPreview, True Float=1, False Float=$System.Timeline.Time)
    )
    ScaleA=true
    ScaleRGB=true
   )
   &ScaleColor(
    Scale Alpha=&MakeCustomFloatFromBool(Bool=$User.HideFootFX, True Float=0, False Float=1)
    ScaleA=true
    ScaleRGB=true
   ) {module=ScaleColor001}
   /SetVariables(Particles.SmokeColor:LinearColor=&MakeLinearColorFromVectorAndFloat(
     "Vector (RGB)"=&MakeVectorFromLinearColorRGB(Color=<rgba 0 0 0 1>)
     "Float (Alpha)"=&FloatFromCurve(
      FloatCurve=DataInterface() {propertyValues {Curve {keys [
          {arriveTangent=4.4759, interpMode=RCIM_Cubic, leaveTangent=4.4759, tangentMode=RCTM_User, time=0, value=0}
          {interpMode=RCIM_Cubic, leaveTangent=-1.8778, tangentMode=RCTM_Break, time=0.2, value=0.8}
          <at 1 0>
         ]}}}
      Scale Curve=&MakeCustomFloatFromBool(
       Bool=$User.isOffPawnPreview
       True Float=&FloatFromCurve(
        FloatCurve=DataInterface() {propertyValues {Curve {keys [
            {interpMode=RCIM_Cubic, leaveTangent=3.1595, tangentMode=RCTM_Break, time=0, value=0}
            {arriveTangent=0.2069, interpMode=RCIM_Cubic, leaveTangent=0.2069, tangentMode=RCTM_User, time=0.2, value=1}
            {interpMode=RCIM_Cubic, tangentMode=RCTM_Break, time=0.5, value=1}
            {arriveTangent=-3.2253, tangentMode=RCTM_User, time=0.8, value=0}
           ]}}}
        CurveIndex=$System.NormalizedLoopAge
       )
       False Float=&MakeCustomFloatFromBool(Bool=$User.HideFootFX, True Float=0, False Float=$System.Timeline.Time)
      )
     )
    )) {isSetParameters=true, module=SetVariables_#7}
   &ScaleMeshSize(Scale Factor=&Multiply_VectorByFloat(
     Vector=<1 1 1>
     Float=&MakeCustomFloatFromBool(
      Bool=&SetBoolByIntComparison(A=$Particles.PartitionParticles.Partition, B=10, Comparison Type=::NewEnumerator6)
      True Float=&MakeFloatFromVector(Vector=$User.ShoeScale_R, Channel=::NewEnumerator0)
      False Float=&MakeFloatFromVector(Vector=$User.ShoeScale_L, Channel=::NewEnumerator0)
     )
    )) {module=ScaleMeshSize001}
  ]
  rendererClasses [&NiagaraMeshRendererProperties]
  renderers [
   {rendererClass=&NiagaraMeshRendererProperties, rendererIndex=0}
   {rendererClass=&NiagaraMeshRendererProperties, rendererIndex=1}
  ]
 }
 SplineFlamesBack {
  emitterUpdateScript [
   &EmitterState(
    Inactive Response=::NewEnumerator1
    Life Cycle Mode=::NewEnumerator1
    Loop Behavior=::NewEnumerator1
    Loop Duration Mode=::NewEnumerator1
    UseLoopDelay=false
   )
   &SpawnBurst_Instantaneous(Spawn Count=6, Use Spawn Probability=false, UseLoopCountLimit=false)
  ]
  particleSpawnScript [
   &PartitionParticles(Number Of Partitions=6)
   &InitializeParticle(
    Color=<rgba 1 0.2359 0.4907 1>
    Lifetime=5
    Mesh Scale=<0.5 0.5 0.65>
    Mesh Scale Mode=::NewEnumerator4
    UsePositionOffset=false
    Write Material Random=true
   )
   &InitialMeshOrientation(
    Mesh Orientation Mode=::NewEnumerator2
    Rotation=&SelectVectorFromArray(
     Direct Array Index=$Particles.PartitionParticles.Partition
     Vector Selection Array=DataInterface() {propertyValues {FloatData [<-0.25 0 0>, <0.25 0 0>, <0.25 0 0>, <-0.25 0 0>, <0 0.25 0>, <0 0.25 0>]}}
    )
    Use Rotation Vector=true
   )
  ]
  particleUpdateScript [
   &ParticleState(Kill Particles When Lifetime Has Elapsed=false)
   &ScaleMeshSize(Scale Factor=&MakeCustomVectorFromBool(
     Bool=$User.isOffPawnPreview
     True Vector=<1 1 1>
     False Vector=&Lerp_Vector(A=<1 1 0>, B=<1 1 1.5>, Alpha=$System.Timeline001.Time)
    ))
   /SetVariables(Particles.SmokeSize:Vector3f=&Lerp_Vector(A=<1 1 0>, B=<1 1 0.9>, Alpha=$System.Timeline001.Time)) {isSetParameters=true, module=SetVariables_#6}
   &SkeletalMeshLocation(
    Bone / Socket Index=&SelectIntFromArray(Direct Array Index=$Particles.PartitionParticles.Partition, Int Selection Array=DataInterface() {propertyValues {IntData [2,3,4,12,13,14]}})
    MaskSpawnGroup=false
    Mesh Sampling Type=::NewEnumerator5
    Sampled Position Offset=&SelectVectorFromArray(
     Direct Array Index=$Particles.PartitionParticles.Partition
     Vector Selection Array=DataInterface() {propertyValues {FloatData [<0.5 -4 1>, <0.5 4 1>, <0.5 4 1>, <0.5 -4 1>, <-5 0 1>, <-5 0 1>]}}
    )
    Skeletal Mesh=$System.Skeletal Mesh
    Socket Sampling Mode=::NewEnumerator5
   )
   &DynamicMaterialParameters(
    Index 0 Param 1=2
    Index 0 Param 2=&SelectFloatFromArray(
     Direct Array Index=$Particles.PartitionParticles.Partition
     Float Selection Array=DataInterface() {propertyValues {FloatData [5,5,5,5,20,20]}}
    )
    Index 0 Param 3=&RandomRangeFloat(Minimum=0, Maximum=1)
    Module.Index0Float0Write=true
    Module.Index0Float1Write=true
    Module.Index0Float2Write=true
    Module.Index0Float3Write=true
   )
   &ScaleColor(
    Scale Alpha=&MakeCustomFloatFromBool(
     Bool=$User.isOffPawnPreview
     True Float=&FloatFromCurve(
      FloatCurve=DataInterface() {propertyValues {Curve {keys [
          {interpMode=RCIM_Cubic, leaveTangentWeight=5.5233, tangentMode=RCTM_Break, time=0, value=0}
          {arriveTangentWeight=1.702, interpMode=RCIM_Cubic, leaveTangentWeight=3.2337, tangentMode=RCTM_User, time=1, value=0}
          {arriveTangentWeight=0.335, interpMode=RCIM_Cubic, leaveTangentWeight=3.015, tangentMode=RCTM_User, time=2, value=1}
          {arriveTangentWeight=3, interpMode=RCIM_Cubic, leaveTangentWeight=1.6667, tangentMode=RCTM_Break, time=3, value=1}
          {arriveTangent=-1.327, tangentMode=RCTM_User, time=4.5, value=0}
          <at 5 0>
         ]}}}
      CurveIndex=&ModuloFloat(A=&Divide_Float(A=$Engine.Time, B=4), B=6)
      Scale Curve=0.5
     )
     False Float=&MakeCustomFloatFromBool(Bool=$User.HideFootFX, True Float=0, False Float=$System.Timeline.Time)
    )
    ScaleA=true
    ScaleRGB=true
   )
   /SetVariables(Particles.SmokeColor:LinearColor=&ColorFromCurve(
     ColorCurve=DataInterface() {propertyValues {
       AlphaCurve {keys [<at 0.0179 0>,<at 0.3723 0.4 RCIM_Cubic>]}
       BlueCurve {keys [<at 1 0 RCIM_Cubic>]}
       GreenCurve {keys [<at 1 0 RCIM_Cubic>]}
       RedCurve {keys [<at 1 0 RCIM_Cubic>]}
      }}
     CurveIndex=&MakeCustomFloatFromBool(Bool=$User.HideFootFX, True Float=0, False Float=$System.Timeline.Time)
    )) {isSetParameters=true, module=SetVariables_#7}
   &ScaleMeshSize(Scale Factor=&VectorFromFloat(Value=&MakeCustomFloatFromBool(
      Bool=&SetBoolByIntComparison(A=$Particles.PartitionParticles.Partition, B=3, Comparison Type=::NewEnumerator6)
      True Float=&MakeFloatFromVector(Vector=$User.ShoeScale_R, Channel=::NewEnumerator0)
      False Float=&MakeFloatFromVector(Vector=$User.ShoeScale_L, Channel=::NewEnumerator0)
     ))) {module=ScaleMeshSize001}
  ]
  rendererClasses [&NiagaraMeshRendererProperties]
  renderers [
   {rendererClass=&NiagaraMeshRendererProperties, rendererIndex=0}
   {rendererClass=&NiagaraMeshRendererProperties, rendererIndex=1}
  ]
 }
}
summary {
 systemName=NS_Shoe_ScorchPoint_Idle
 emitters [
  {emitterName=Embers, rendererClasses [&NiagaraSpriteRendererProperties]}
  {bEnabled=false, emitterName=GPU_Flames, simTarget=GPUComputeSim, rendererClasses [&NiagaraSpriteRendererProperties]}
  {emitterName=GlowAndLights, rendererClasses [&NiagaraLightRendererProperties, &NiagaraSpriteRendererProperties]}
  {emitterName=SplineFlamesBack, rendererClasses [&NiagaraMeshRendererProperties]}
  {emitterName=SplineFlames, rendererClasses [&NiagaraMeshRendererProperties]}
  {emitterName=GPU_Flames002, rendererClasses [&NiagaraSpriteRendererProperties]}
 ]
 userVariables [
  User.bIsFrontEnd {type=NiagaraBool}
  User.bIsFrontEndPreview {type=NiagaraBool}
  User.bisLit {type=NiagaraBool}
  User.HideFootFX {type=NiagaraBool}
  User.isOffPawnPreview {type=NiagaraBool}
  User.RibbonIDOffset {type=NiagaraInt32}
  User.ShoeScale_L {type=Vector3f, description="", defaultValue {struct=Vector3f, value=<0 0 0>}}
  User.ShoeScale_R {type=Vector3f, description="", defaultValue {struct=Vector3f, value=<0 0 0>}}
 ]
}
