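# Material function for the ScorchPoint cosmetic shoe that layers staged, animated lava effects over incoming material attributes, changing base color, specular, roughness and emissive.
# CurrentStage, or a looping cycle in the frontend, drives stage and cooldown blends, with transitions timed from GameTimeFromBP.
# Packed mask and SDF textures define three elements: pulsing lines, molten lava and bump detail.
# Flipbook, panning and triplanar noise, gradient lookups, Fresnel and pre-skinned position gradients shape them.
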MaterialFunction /CosmeticShoes/Assets/FORT_Shoes/Shoe_ScorchPoint/Materials/MF_ScorchPoint_Shoe
graph {nodes [
  ConstantBiasScale(None=Sine(None=NamedRerouteUsage()))
  Fmod(A=Multiply(A=NamedRerouteUsage()) {constB=0.5}, B=Constant() {r=6})
  n1 = GetMaterialAttributes(None=FunctionInput() {inputName="Material (MA)", inputType=FunctionInput_MaterialAttributes}) {attributeGetTypes [
    #1
    #2
    #3
    #4
    #5
    #6
   ]}
  n2 = SetMaterialAttributes!(
   MaterialAttributes=@n1.MaterialAttributes
   BaseColor=NamedRerouteUsage()
   Metallic=@n1.MaterialAttributes
   Specular=NamedRerouteUsage()
   Roughness=NamedRerouteUsage()
   Normal=@n1.MaterialAttributes
   EmissiveColor=NamedRerouteUsage()
  ) {attributeSetTypes [
    #1
    #2
    #3
    #4
    #5
    #6
   ]}
  FunctionOutput(None=@n2)
  n3 = NamedRerouteUsage()
  LinearInterpolate(A=@n3, Alpha=NamedRerouteUsage())
  n4 = MaterialFunctionCall(Y=Multiply(A=NamedRerouteUsage(), B=ScalarParameter() {defaultValue=0.25, parameterName=Flame_Speed}) {constB=0.25}) {materialFunction=&MakeFloat2}
  n5 = Power(Base=MaterialFunctionCall!(
    Export Float 4
    TextureObject=TextureObject() {samplerType=SAMPLERTYPE_LinearGrayscale, texture=/Game/Global/Textures/Default/Noise/T_SwirlNoise_H2}
    TextureSize=Constant() {r=10}
    WorldPosition=NamedRerouteUsage()
    World Space Normal=VertexNormalWS()
    ProjectionTransitionContrast
    Pan X=@n4.Result
    Pan Y=@n4.Result
    Pan Z
   ) {materialFunction=/BRCosmetics/Global/MaterialFunctions/TechArt/TriPlanerProjection_SeperatePanningDirections}.XY Texture)
  Multiply(A=@n5, B=NamedRerouteUsage())
  n6 = TextureSample() {texture=/CosmeticShoes/Assets/FORT_Shoes/Shoe_ScorchPoint/Textures/T_ScorchPoint_Shoe_Reactive_Msk}
  NamedRerouteDeclaration(None=@n6.R) {name=Stage_1_Msk, nodeColor=<rgba 0 1 0.0908 1>}
  NamedRerouteDeclaration(None=@n6.G) {name=Stage_2_Msk, nodeColor=<rgba 0 1 0.0908 1>}
  NamedRerouteDeclaration(None=@n1.Normal) {name=IN_Normal, nodeColor=<rgba 0 0.5078 1 1>}
  n7 = Time()
  n8 = Subtract(A=@n7, B=ScalarParameter() {defaultValue=-99999, group=ScorchPoint, parameterName=GameTimeFromBP, sortPriority=100})
  n9 = Add(A=ScalarParameter() {defaultValue=1, group=ScorchPoint, parameterName=Animation_Duration, sortPriority=101}) {constB=0}
  NamedRerouteDeclaration(None=Saturate(None=Divide(A=@n8, B=@n9))) {name=TimeNative, nodeColor=<rgba 0.2891 1 0 1>}
  n10 = NamedRerouteUsage()
  n11 = LinearInterpolate(A=@n10, B=NamedRerouteUsage(), Alpha=Subtract(A=NamedRerouteUsage(), B=NamedRerouteUsage()))
  n12 = NamedRerouteUsage()
  n13 = Multiply(A=Multiply(A=NamedRerouteUsage()) {constB=10}, B=NamedRerouteUsage())
  n14 = NamedRerouteUsage()
  n15 = Add(
   A=LinearInterpolate(A=@n10, B=Add(A=@n11, B=@n12), Alpha=NamedRerouteUsage())
   B=Multiply(A=NamedRerouteUsage(), B=Multiply(A=@n13, B=@n14) {constB=10})
  ) {desc=Add Bump details}
  n16 = NamedRerouteUsage()
  NamedRerouteDeclaration(None=Multiply(A=Multiply(A=@n15, B=@n16), B=NamedRerouteUsage())) {name=Stage_3_Element, nodeColor=<rgba 0.4531 0 1 1>}
  n17 = NamedRerouteUsage()
  n18 = ScalarParameter() {defaultValue=0.0625, group=ScorchPoint, parameterName=Lava_Turbulence_Speed}
  n19 = MaterialFunctionCall!(
   Size=Constant2Vector() {g=8, r=16}
   Selector=Modulo!(A=Multiply(A=@n17, B=@n18) {constB=0.0625}, B=Constant() {r=16})
   UV=TextureCoordinate()
  ) {materialFunction=/Game/Global/MaterialFunctions/MF_FlipbookUV}
  n20 = TextureObject() {texture=/CosmeticShoes/Assets/FORT_Shoes/Shoe_ScorchPoint/Textures/T_ScorchPoint_Shoe_Lava_Noise}
  n21 = SmoothStep(Value=LinearInterpolate(A=TextureSample(UVs=@n19.FlipbookUV1, Tex=@n20).R, B=TextureSample(UVs=@n19.FlipbookUV2, Tex=@n20).R, Alpha=@n19.Blend2)) {constMax=3, constMin=0.1}
  NamedRerouteDeclaration(None=Saturate(None=Power(Base=@n21) {constExponent=0.25})) {name=LavaNoise, nodeColor=<rgba 0 0.9062 1 1>}
  n22 = NamedRerouteUsage()
  n23 = Add(A=Multiply(A=NamedRerouteUsage(), B=OneMinus(None=@n22)), B=NamedRerouteUsage()) {desc="Add back og emissive that not at shoe sole, e.g. logo"}
  n24 = Multiply(A=OneMinus(None=NamedRerouteUsage()), B=NamedRerouteUsage())
  n25 = NamedRerouteUsage()
  n26 = Add(A=@n23, B=Multiply(A=NamedRerouteUsage(), B=SmoothStep(Max=@n24, Value=@n25)))
  n27 = OneMinus(None=NamedRerouteUsage())
  n28 = NamedRerouteUsage()
  n29 = Multiply(A=NamedRerouteUsage(), B=SmoothStep(Max=Multiply(A=@n27, B=@n28), Value=NamedRerouteUsage()))
  NamedRerouteDeclaration(None=LinearInterpolate(A=Add(A=@n26, B=@n29), B=@n23, Alpha=NamedRerouteUsage())) {name=OUT_Emissive, nodeColor=<rgba 1 0.6094 0 1>}
  NamedRerouteDeclaration(None=@n11) {name=LavaBC, nodeColor=<rgba 0 1 0.7656 1>}
  n30 = NamedRerouteUsage()
  n31 = LinearInterpolate(
   A=ScalarParameter() {group=ScorchPoint, parameterName=CurrentStage, sortPriority=3}
   B=Fmod(A=Divide(A=@n30) {constB=4, desc=4 times slower}, B=Constant() {r=6}) {desc=0 -6 loop}
   Alpha=ScalarParameter() {group=ScorchPoint, parameterName=IsShoeInFrontend}
  )
  n32 = LinearInterpolate(A=Subtract(A=@n31), B=@n31, Alpha=NamedRerouteUsage()) {desc=Native lerp toward}
  n33 = Constant4Vector() {desc=Stage 1 to 2, constant=<rgba 1 2 0 1>}
  NamedRerouteDeclaration(None=Saturate(None=MaterialFunctionCall(Input=@n32, Input Low=@n33.R, Input High=@n33.R, Target Low=@n33.R, Target High=@n33.R) {materialFunction=&RemapValueRange}.Result)) {name=Stage2_Lerp, nodeColor=<rgba 0.1893 1 0 1>}
  n34 = Constant4Vector() {desc=Stage 2 to 3, constant=<rgba 2 3 0 1>}
  NamedRerouteDeclaration(None=Saturate(None=MaterialFunctionCall(Input=@n32, Input Low=@n34.R, Input High=@n34.R, Target Low=@n34.R, Target High=@n34.R) {materialFunction=&RemapValueRange}.Result)) {name=Stage3_Lerp, nodeColor=<rgba 0.1893 1 0 1>}
  n35 = Constant4Vector() {desc=Stage 0 to 1, constant=<rgba 0 1 0 1>}
  NamedRerouteDeclaration(None=Saturate(None=MaterialFunctionCall(Input=@n32, Input Low=@n35.R, Input High=@n35.R, Target Low=@n35.R, Target High=@n35.R) {materialFunction=&RemapValueRange}.Result)) {name=Stage1_Lerp, nodeColor=<rgba 0.1893 1 0 1>}
  NamedRerouteDeclaration(None=@n1.EmissiveColor) {name=IN_Emissive, nodeColor=<rgba 1 0.1406 0 1>}
  NamedRerouteDeclaration(None=@n6.B) {name=Stage_3_Msk, nodeColor=<rgba 0 1 0.0908 1>}
  n36 = NamedRerouteUsage()
  n37 = ComponentMask(None=@n36) {b=true}
  n38 = OneMinus(None=SmoothStep(Value=@n37) {constMax=8})
  NamedRerouteDeclaration(None=Multiply(A=Power(Base=@n38), B=NamedRerouteUsage())) {name=Up_Gradient_Msk, nodeColor=<rgba 1 0.7734 0 1>}
  NamedRerouteDeclaration(None=VectorParameter() {group=ScorchPoint, parameterName=Lava_Color_03, defaultValue=<rgba 0.916 0.0256 0.0641 1>}.RGB) {name=TintColor_03, nodeColor=<rgba 1 1 1 1>}
  n39 = NamedRerouteUsage()
  n40 = CameraVectorWS()
  n41 = OneMinus(None=DotProduct(A=@n39, B=@n40))
  NamedRerouteDeclaration(None=Power(Base=Saturate(None=@n41), Exp=Constant() {r=1})) {name=Stage_1_Fresnel, nodeColor=<rgba 1 0.3516 0 1>}
  NamedRerouteDeclaration(None=VectorParameter() {group=ScorchPoint, parameterName=Lava_Color_02, defaultValue=<rgba 1 0.3822 0 1>}.RGB) {name=TintColor_02, nodeColor=<rgba 1 1 1 1>}
  n42 = TextureObjectParameter() {
   group=ScorchPoint
   parameterName=Lava_Gradient
   samplerSource=SSM_Clamp_WorldGroupSettings
   texture=/CosmeticShoes/Assets/FORT_Shoes/Shoe_ScorchPoint/Textures/T_ScorchPoint_Shoe_Lava_Gradient
  }
  n43 = TextureSample(UVs=Clamp(None=NamedRerouteUsage()) {maxDefault=0.95, minDefault=0.1}, Tex=@n42) {samplerSource=SSM_Clamp_WorldGroupSettings}
  NamedRerouteDeclaration(None=@n43.RGB) {name=TintedLavaNoise, nodeColor=<rgba 0.2891 1 0 1>}
  NamedRerouteDeclaration(None=@n7) {name=MyTime, nodeColor=<rgba 1 0.3281 0 1>}
  n44 = Constant4Vector() {desc=Stage 4 to 5, constant=<rgba 4 5 0 1>}
  NamedRerouteDeclaration(None=Saturate(None=MaterialFunctionCall(Input=@n32, Input Low=@n44.R, Input High=@n44.R, Target Low=@n44.R, Target High=@n44.R) {materialFunction=&RemapValueRange}.Result)) {name=Cooldown_Lerp, nodeColor=<rgba 0.1893 1 0 1>}
  NamedRerouteDeclaration(None=@n1.Specular) {name=IN_Specular, nodeColor=<rgba 0.0312 0 1 1>}
  n45 = NamedRerouteUsage()
  n46 = Constant() {r=1}
  n47 = Multiply(A=NamedRerouteUsage(), B=NamedRerouteUsage())
  n48 = NamedRerouteUsage()
  n49 = NamedRerouteUsage()
  n50 = NamedRerouteUsage()
  n51 = LinearInterpolate(
   A=LinearInterpolate(A=@n45, B=@n46, Alpha=Multiply(A=@n47, B=@n48))
   B=@n46
   Alpha=Multiply(A=Multiply(A=@n49, B=@n50), B=NamedRerouteUsage())
  )
  n52 = NamedRerouteUsage()
  NamedRerouteDeclaration(None=LinearInterpolate(A=LinearInterpolate(A=@n51, B=@n45, Alpha=@n52), B=@n45, Alpha=NamedRerouteUsage())) {name=OUT_Specular, nodeColor=<rgba 0 1 0.0859 1>}
  n53 = TextureSample() {texture=/CosmeticShoes/Assets/FORT_Shoes/Shoe_ScorchPoint/Textures/T_ScorchPoint_Shoe_Reactive_SDF}
  NamedRerouteDeclaration(None=@n53.R) {name=Stage_1_Element_SDF, nodeColor=<rgba 0.131 0.3144 1 1>}
  NamedRerouteDeclaration(None=VectorParameter() {group=ScorchPoint, parameterName=Lava_Color_01, defaultValue=<rgba 1 0 0 1>}.RGB) {name=TintColor_01, nodeColor=<rgba 1 1 1 1>}
  NamedRerouteDeclaration(None=@n1.Roughness) {name=IN_Roughness, nodeColor=<rgba 0 0.5547 1 1>}
  n54 = NamedRerouteUsage()
  n55 = Constant() {r=1}
  n56 = Multiply(A=NamedRerouteUsage(), B=NamedRerouteUsage())
  n57 = NamedRerouteUsage()
  n58 = NamedRerouteUsage()
  n59 = NamedRerouteUsage()
  n60 = LinearInterpolate(
   A=LinearInterpolate(A=@n54, B=@n55, Alpha=Multiply(A=@n56, B=@n57))
   B=@n55
   Alpha=Multiply(A=Multiply(A=@n58, B=@n59), B=NamedRerouteUsage())
  )
  n61 = NamedRerouteUsage()
  NamedRerouteDeclaration(None=LinearInterpolate(A=LinearInterpolate(A=@n60, B=@n54, Alpha=@n61), B=@n54, Alpha=NamedRerouteUsage())) {name=OUT_Roughness, nodeColor=<rgba 1 0 0.1641 1>}
  n62 = MaterialFunctionCall(Float3=FunctionInput() {inputName=M, sortPriority=1}) {desc="R = AO\r\nG = GMap\r\nB = Skin", materialFunction=&BreakOutFloat3Components}
  NamedRerouteDeclaration(None=@n62.R) {name=Ambient Occlusion, nodeColor=<rgba 0.875 0 1 1>}
  NamedRerouteDeclaration(None=@n62.G) {name=GMap, nodeColor=<rgba 0.6641 0 1 1>}
  NamedRerouteDeclaration(None=@n5) {name=TriplannarNoise, nodeColor=<rgba 1 0 0.1641 1>}
  n63 = Power(Base=SmoothStep(Value=@n37) {constMax=4, constMin=-4}) {constExponent=2.5}
  n64 = NamedRerouteUsage()
  NamedRerouteDeclaration(None=Multiply(A=Multiply(A=@n63, B=@n64), B=NamedRerouteUsage())) {name=Up_Gradient_Msk_02, nodeColor=<rgba 1 0.7734 0 1>}
  n65 = Abs(None=ComponentMask(None=@n36) {g=true})
  NamedRerouteDeclaration(None=Power(Base=SmoothStep(Value=@n65) {constMax=10})) {name=Front_Gradient_Msk, nodeColor=<rgba 1 0.7734 0 1>}
  NamedRerouteDeclaration(None=Multiply(A=SmoothStep(Value=@n37) {constMax=4, constMin=-2}, B=NamedRerouteUsage())) {name=Up_Gradient_Msk_03, nodeColor=<rgba 1 0.7734 0 1>}
  n66 = SmoothStep(Value=@n37) {constMax=5, constMin=-1}
  NamedRerouteDeclaration(None=Multiply(A=Power(Base=@n66), B=NamedRerouteUsage())) {name=Up_Gradient_Msk_04, nodeColor=<rgba 1 0.7734 0 1>}
  NamedRerouteDeclaration(None=@n53.G) {name=Stage_1_Element_timeoffset, nodeColor=<rgba 0.131 0.3144 1 1>}
  n67 = DotProduct(A=@n39, B=@n40)
  NamedRerouteDeclaration(None=Power(Base=Saturate(None=@n67), Exp=Constant() {r=20})) {name=Stage_2_Fresnel, nodeColor=<rgba 0.5469 0 1 1>}
  NamedRerouteDeclaration(None=@n1.BaseColor) {name=IN_BC, nodeColor=<rgba 0.2891 0 1 1>}
  n68 = NamedRerouteUsage()
  n69 = Add(A=ComponentMask(None=@n68) {r=true}, B=ComponentMask(None=@n68) {b=true})
  NamedRerouteDeclaration(None=LinearInterpolate(A=@n3, B=NamedRerouteUsage(), Alpha=Saturate(None=@n69))) {name=OUT_BC, nodeColor=<rgba 0.0312 1 0 1>}
  n70 = MaterialFunctionCall(Float3=@n39) {materialFunction=&BreakOutFloat3Components}
  n71 = DotProduct(A=@n70.R, B=@n70.R)
  n72 = DotProduct(A=@n70.R, B=@n70.R)
  NamedRerouteDeclaration(None=Max(A=Max(A=@n71, B=@n72), B=DotProduct(A=@n70.G, B=@n70.G))) {name=Stage_3_Bumps, nodeColor=<rgba 0.0637 1 0.9245 1>}
  NamedRerouteDeclaration(None=Transform(None=TextureSampleParameter2D() {parameterName=Normals, samplerType=SAMPLERTYPE_Normal, texture=&FlatNormalBlue}.RGB)) {name=ReusedNormal, nodeColor=<rgba 0 0.0156 1 1>}
  NamedRerouteDeclaration(None=@n53.B) {name=Stage_3_Element_SDF, nodeColor=<rgba 0.131 0.3144 1 1>}
  NamedRerouteDeclaration(None=Multiply(A=@n53.R, B=@n53.R)) {name=Stage_2_Element_SDF, nodeColor=<rgba 0.131 0.3144 1 1>}
  n73 = NamedRerouteUsage()
  n74 = Constant()
  n75 = NamedRerouteUsage()
  n76 = Add(
   A=Multiply(A=NamedRerouteUsage(), B=ScalarParameter() {defaultValue=0.325, group=ScorchPoint, parameterName=LavaLine_Pulse_Speed})
   B=NamedRerouteUsage()
  ) {constB=0}
  n77 = SmoothStep(Min=@n74, Max=@n75, Value=OneMinus(None=Frac(None=@n76)))
  n78 = Add(A=@n76, B=Constant() {r=0.5})
  n79 = SmoothStep(Min=@n74, Max=@n75, Value=OneMinus(None=Frac(None=@n78)))
  n80 = OneMinus(None=@n75)
  n81 = SmoothStep(Min=@n74, Max=@n80, Value=Frac(None=@n76))
  n82 = SmoothStep(Min=@n74, Max=@n80, Value=Frac(None=@n78))
  n83 = Min(A=OneMinus(None=Multiply(A=@n77, B=@n79)), B=OneMinus(None=Multiply(A=@n81, B=@n82)))
  n84 = TextureObjectParameter() {
   group=ScorchPoint
   parameterName=Lava_Gradient
   texture=&T_ScorchPoint_Shoe_Lava_Gradient
  }
  n85 = TextureSample(
   UVs=SmoothStep(Max=LinearInterpolate(Alpha=@n73) {constA=5, constB=2}, Value=Multiply(A=@n83) {constB=10, desc=LavaLineImpuse}) {constMax=10, constMin=0.1}
   Tex=@n84
  ) {samplerSource=SSM_Clamp_WorldGroupSettings}
  n86 = Multiply(A=@n85.R, B=NamedRerouteUsage())
  n87 = Add(A=@n86) {constB=0.1}
  n88 = Multiply(A=LinearInterpolate(A=Multiply(A=@n85.RGB, B=@n87), B=NamedRerouteUsage(), Alpha=@n86) {desc=ColorStrategy for Next Stage}) {constB=3}
  n89 = NamedRerouteUsage()
  n90 = Add(
   A=Multiply(A=NamedRerouteUsage(), B=NamedRerouteUsage()) {constB=0.1, desc=base read line}
   B=LinearInterpolate(A=LinearInterpolate(A=@n85.RGB, B=@n88, Alpha=@n89) {desc=Blend to next stage}, B=@n85.RGB, Alpha=NamedRerouteUsage()) {desc=back to initial stage}
  ) {desc=Line add base red line}
  n91 = NamedRerouteUsage()
  NamedRerouteDeclaration(None=Multiply(A=Multiply(A=@n90, B=@n91), B=ScalarParameter() {defaultValue=1, group=ScorchPoint, parameterName=LavaLine_Intenstiy})) {name=Stage_1_Element, nodeColor=<rgba 0.4531 0 1 1>}
  n92 = NamedRerouteUsage()
  n93 = NamedRerouteUsage()
  n94 = Multiply(A=NamedRerouteUsage(), B=NamedRerouteUsage())
  n95 = TextureCoordinate()
  n96 = Panner(Coordinate=Multiply(A=@n95) {constB=2}) {speedX=0.02, speedY=0.02}
  n97 = TextureObject() {samplerType=SAMPLERTYPE_LinearGrayscale, texture=&T_SwirlNoise_H2}
  n98 = Panner(Coordinate=@n95) {speedX=-0.02, speedY=-0.02}
  n99 = Max(
   A=Max(
    A=TextureSample(UVs=@n96, Tex=@n97) {samplerSource=SSM_Wrap_WorldGroupSettings, samplerType=SAMPLERTYPE_Grayscale}.B
    B=TextureSample(UVs=@n98, Tex=@n97) {samplerSource=SSM_Wrap_WorldGroupSettings, samplerType=SAMPLERTYPE_Grayscale}.B
   )
   B=NamedRerouteUsage()
  )
  n100 = NamedRerouteUsage()
  n101 = NamedRerouteDeclaration(None=Saturate(None=LinearInterpolate(A=@n94, B=@n99, Alpha=@n100))) {name=LavaBase, nodeColor=<rgba 0.0312 0 1 1>}
  n102 = LinearInterpolate(A=@n92, B=@n93, Alpha=@n101) {desc=Tinted Lava Base}
  n103 = LinearInterpolate(A=@n102, B=Add(A=@n102, B=@n43.RGB), Alpha=@n101)
  n104 = Add(A=NamedRerouteUsage(), B=NamedRerouteUsage())
  n105 = LinearInterpolate(
   A=LinearInterpolate(A=@n103, B=LinearInterpolate(A=@n103, B=@n93, Alpha=@n104), Alpha=NamedRerouteUsage()) {constAlpha=1, desc=stage 1 Color}
   B=@n103
   Alpha=NamedRerouteUsage()
  )
  n106 = NamedRerouteUsage()
  NamedRerouteDeclaration(None=Multiply(A=LinearInterpolate(A=@n105, B=@n92, Alpha=@n106) {constB=0}, B=NamedRerouteUsage())) {name=Stage_2_Element, nodeColor=<rgba 0.4531 0 1 1>}
  NamedRerouteDeclaration(None=VertexInterpolator(VS=LocalPosition() {localOrigin=InstancePreSkinning}.XYZ).PS) {name=PreskinnedLoc, nodeColor=<rgba 0.4895 0.0228 0.9323 1>}
  NamedRerouteUsage()
 ]}
